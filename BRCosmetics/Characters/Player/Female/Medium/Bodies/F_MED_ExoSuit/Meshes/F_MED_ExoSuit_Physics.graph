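# Physics asset for the female medium ExoSuit body, setting up dangling rope physics beside the left thigh.
# Front and back chains of small capsules hang from sphere bodies on their base bones, linked by angle-limited constraints.
# A final constraint with free swing and twist ties the two chain tips together, closing the rope into a loop.
# The left thigh has an unconstrained sphere and capsule body, apparently as a collider for the rope.
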PhysicsAsset /BRCosmetics/Characters/Player/Female/Medium/Bodies/F_MED_ExoSuit/Meshes/F_MED_ExoSuit_Physics
bodies {
 dyn_rope_01_bk [{shapeType=Capsule, length=3.7181, radius=1.396, shapeName=dyn_rope_01_bk_capsule, center=<2.7906 0 0.0001>, rotation=<rot 90 0 0>}]
 dyn_rope_01_frnt [{shapeType=Capsule, length=5.6848, radius=1.7429, shapeName=dyn_rope_01_frnt_capsule, center=<3.8766 0.6352 0>, rotation=<rot 90 0 -9.4703>}]
 dyn_rope_02_bk [{shapeType=Capsule, length=5.236, radius=1.1844, shapeName=dyn_rope_02_bk_capsule, center=<3.2828 0 0>, rotation=<rot 90 0 0>}]
 dyn_rope_02_frnt [{
   shapeType=Capsule
   length=3.4841
   radius=2.1399
   shapeName=dyn_rope_02_frnt_capsule
   center=<3.4785 0.8676 0>
   rotation=<rot 90 14.0363 -2.8221>
  }]
 dyn_rope_03_frnt [{shapeType=Capsule, length=1.75, radius=1.9991, shapeName=dyn_rope_03_frnt_capsule, center=<1.5768 0 0>, rotation=<rot 90 0 0>}]
 dyn_rope_base_bk [{shapeType=Sphere, radius=0.5, shapeName=dyn_rope_base_bk_sphere}]
 dyn_rope_base_frnt [{shapeType=Sphere, radius=0.5, shapeName=dyn_rope_base_frnt_sphere}]
 dyn_rope_parent_bk [{shapeType=Capsule, length=3.56, radius=1.193, shapeName=dyn_rope_parent_bk_capsule, center=<2.6022 0 0.0001>, rotation=<rot 90 0 0>}]
 dyn_rope_parent_frnt [{shapeType=Capsule, length=3.2118, radius=1.6665, shapeName=dyn_rope_parent_frnt_capsule, center=<2.6999 0 0>, rotation=<rot 90 0 0>}]
 thigh_l [
  {shapeType=Sphere, radius=11, shapeName=thigh_l_sphere, center=<-7.6404 -1.7819 5.626>}
  {
   shapeType=Capsule
   length=28.5
   radius=7.25
   shapeName=thigh_l_capsule
   center=<-20.9293 -0.4792 1.7454>
   rotation=<rot 89.4092 176.7032 179.9996>
  }
 ]
}
constraints [
 {bone1Name=dyn_rope_parent_bk, bone2Name=dyn_rope_base_bk, swing1LimitDegrees=100, swing2LimitDegrees=100, twistLimitDegrees=50}
 {bone1Name=dyn_rope_01_bk, bone2Name=dyn_rope_parent_bk, swing1LimitDegrees=80, swing2LimitDegrees=80, twistLimitDegrees=50}
 {bone1Name=dyn_rope_02_bk, bone2Name=dyn_rope_01_bk, swing1LimitDegrees=80, swing2LimitDegrees=80, twistLimitDegrees=10}
 {bone1Name=dyn_rope_parent_frnt, bone2Name=dyn_rope_base_frnt, swing1LimitDegrees=60, swing2LimitDegrees=90, twistLimitDegrees=100}
 {bone1Name=dyn_rope_01_frnt, bone2Name=dyn_rope_parent_frnt, swing1LimitDegrees=80, swing2LimitDegrees=40, twistLimitDegrees=10}
 {bone1Name=dyn_rope_02_frnt, bone2Name=dyn_rope_01_frnt, swing1LimitDegrees=90, swing2LimitDegrees=60, twistLimitDegrees=10}
 {bone1Name=dyn_rope_03_frnt, bone2Name=dyn_rope_02_frnt, swing1LimitDegrees=60, swing2LimitDegrees=20, twistLimitDegrees=10}
 {
  bone1Name=dyn_rope_03_frnt
  bone2Name=dyn_rope_02_bk
  swing1LimitDegrees=45
  swing1Motion=Free
  swing2LimitDegrees=45
  swing2Motion=Free
  twistLimitDegrees=45
  twistMotion=Free
 }
]
