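# Post-process-style AnimBP for the female medium ExoSuit body cosmetic, adjusting armor, shoulder pads, shield and cape socket.
# It copies the parent mesh pose, chains mostly additive bone offsets, then runs RigidBody physics.
# Updates derive blend alphas from pawn state, weapon type and arm/clavicle twist; both shoulder-pad twist drivers read the left upper arm.
# On customization load, an ExoSuit cape backpack tag enables socket adjustment and hides a material element.
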AnimBlueprint /BRCosmetics/Characters/Player/Female/Medium/Bodies/F_MED_ExoSuit/Meshes/F_MED_ExoSuit_AnimBP
target_skeleton=&/F_MED_ExoSuit_Skeleton
graphs [
 AnimGraph {nodes [
   n1 = Get GliderPose_Alpha() -> (GliderPose_Alpha:double)
   n2 = Get IsInFrontend() -> (IsInFrontend:double)
   n3 = "Transform (Modify) Bone - Bone: dyn_shield_parent"(
    ComponentPose=Local To Component(LocalPose=Copy Pose From Mesh() {bUseAttachedParent=true})
    Alpha=@n2
    Translation="0,0,0"
    Rotation="0,0,15"
    Scale="1,1,1"
   ) {
    rotationMode=BMM_Additive
    rotationSpace=BCS_BoneSpace
    translationMode=BMM_Additive
    translationSpace=BCS_BoneSpace
    alphaBoolBlend {blendOption=Linear}
    boneToModify {boneName=dyn_shield_parent}
    rotation {roll=15}
   }
   n4 = "Transform (Modify) Bone - Bone: dyn_back_armor_wing_r"(ComponentPose=@n3, Alpha=@n2, Translation="0.000,0.000,-1", Rotation="-5,5,-0.000", Scale="1,1,1") {
    rotationMode=BMM_Additive
    rotationSpace=BCS_BoneSpace
    translationMode=BMM_Additive
    translationSpace=BCS_BoneSpace
    alphaBoolBlend {blendOption=Linear}
    boneToModify {boneName=dyn_back_armor_wing_r}
    rotation {pitch=-5, yaw=5}
    translation {z=-1}
   }
   n5 = Get Skydive_Armor_Alpha() -> (Skydive_Armor_Alpha:double)
   n6 = "Transform (Modify) Bone - Bone: dyn_back_armor_wing_l"(ComponentPose=@n4, Alpha=@n5, Translation="-0.326,-0.048,-1.040", Rotation="-24.341,-2.860,6.222", Scale="1,1,1") {
    rotationMode=BMM_Additive
    rotationSpace=BCS_BoneSpace
    translationMode=BMM_Additive
    translationSpace=BCS_BoneSpace
    alphaBoolBlend {blendOption=Linear}
    boneToModify {boneName=dyn_back_armor_wing_l}
    rotation=<rot -24.341 -2.86 6.222>
    translation=<-0.326 -0.048 -1.04>
   }
   n7 = "Transform (Modify) Bone - Bone: dyn_back_armor_wing_r"(ComponentPose=@n6, Alpha=@n5, Translation="-0.386,-0.359,-1.299", Rotation="-20.459,3.042,6.792", Scale="1,1,1") {
    rotationMode=BMM_Additive
    rotationSpace=BCS_BoneSpace
    translationMode=BMM_Additive
    translationSpace=BCS_BoneSpace
    alphaBoolBlend {blendOption=Linear}
    boneToModify {boneName=dyn_back_armor_wing_r}
    rotation=<rot -20.459 3.042 6.792>
    translation=<-0.386 -0.359 -1.299>
   }
   n8 = "Transform (Modify) Bone - Bone: dyn_back_armor_center"(ComponentPose=@n7, Alpha=@n5, Translation="-0.000,-0.000,-0.546", Rotation="3.768,0.000,0.000", Scale="1,1,1") {
    rotationMode=BMM_Additive
    rotationSpace=BCS_BoneSpace
    translationMode=BMM_Additive
    translationSpace=BCS_BoneSpace
    alphaBoolBlend {blendOption=Linear}
    boneToModify {boneName=dyn_back_armor_center}
    rotation {pitch=3.768}
    translation {z=-0.546}
   }
   n9 = "Transform (Modify) Bone - Bone: dyn_back_armor_wing_l"(ComponentPose=@n8, Alpha=@n1, Translation="0,0,0", Rotation="30,0.000,0.000", Scale="1,1,1") {
    rotationMode=BMM_Additive
    rotationSpace=BCS_BoneSpace
    translationMode=BMM_Additive
    translationSpace=BCS_BoneSpace
    alphaBoolBlend {blendOption=Linear}
    boneToModify {boneName=dyn_back_armor_wing_l}
    rotation {pitch=30}
   }
   n10 = "Transform (Modify) Bone - Bone: dyn_back_armor_wing_r"(ComponentPose=@n9, Alpha=@n1, Translation="0,0,0", Rotation="30,0,0", Scale="1,1,1") {
    rotationMode=BMM_Additive
    rotationSpace=BCS_BoneSpace
    translationMode=BMM_Additive
    translationSpace=BCS_BoneSpace
    alphaBoolBlend {blendOption=Linear}
    boneToModify {boneName=dyn_back_armor_wing_r}
    rotation {pitch=30}
   }
   n11 = "Transform (Modify) Bone - Bone: dyn_snake_r"!(
    ComponentPose=@n10
    bAlphaBoolEnabled:bool=OR Boolean(A=Get bIsTargeting(), B=Get Alt_weapon() -> (Alt_weapon:bool))
    Translation="0,0,0"
    Rotation="-20,-30,20"
    Scale="1,1,1"
   ) -> (Pose) {
    alphaInputType=Bool
    rotationMode=BMM_Additive
    rotationSpace=BCS_BoneSpace
    translationMode=BMM_Additive
    translationSpace=BCS_BoneSpace
    alphaBoolBlend {blendInTime=0.1, blendOption=Linear, blendOutTime=0.5}
    alphaScaleBiasClamp {bInterpResult=true, interpSpeedDecreasing=20, interpSpeedIncreasing=20}
    boneToModify {boneName=dyn_snake_r}
    rotation=<rot -20 -30 20>
   }
   n12 = "Transform (Modify) Bone - Bone: dyn_cape_socket_adjustment"!(
    ComponentPose=@n11
    bAlphaBoolEnabled:bool=Get ExoSuitCape_IsEquipped() -> (ExoSuitCape_IsEquipped:bool)
    Translation="-4.756,6.8,-0.013"
    Rotation="0,0,0"
    Scale="1,1,1"
   ) -> (Pose) {
    alphaInputType=Bool
    translationMode=BMM_Replace
    translationSpace=BCS_BoneSpace
    alphaBoolBlend {blendOption=Linear}
    boneToModify {boneName=dyn_cape_socket_adjustment}
    translation=<-4.756 6.8 -0.013>
   }
   n13 = "Transform (Modify) Bone - Bone: dyn_shoulderpad_r"(
    ComponentPose=@n12
    Alpha=Get ArmRotation_R() -> (ArmRotation_R:double)
    Translation="1.000,-1,-0.500"
    Rotation="0.000,0.000,10"
    Scale="1,1,1"
   ) {
    rotationMode=BMM_Additive
    rotationSpace=BCS_BoneSpace
    translationMode=BMM_Additive
    translationSpace=BCS_BoneSpace
    alphaBoolBlend {blendOption=Linear}
    alphaScaleBiasClamp {bInterpResult=true, interpSpeedDecreasing=20, interpSpeedIncreasing=20}
    boneToModify {boneName=dyn_shoulderpad_r}
    rotation {roll=10}
    translation=<1 -1 -0.5>
   }
   n14 = "Transform (Modify) Bone - Bone: dyn_back_armor_wing_l"(ComponentPose=@n13, Alpha=Get ShoulderRotation_L() -> (ShoulderRotation_L:double), Translation="0,0.25,0", Rotation="-12,0,0", Scale="1,1,1") {
    rotationMode=BMM_Additive
    rotationSpace=BCS_BoneSpace
    translationMode=BMM_Additive
    translationSpace=BCS_BoneSpace
    alphaBoolBlend {blendOption=Linear}
    alphaScaleBiasClamp {bInterpResult=true, interpSpeedDecreasing=20, interpSpeedIncreasing=20}
    boneToModify {boneName=dyn_back_armor_wing_l}
    rotation {pitch=-12}
    translation {y=0.25}
   }
   n15 = "Transform (Modify) Bone - Bone: dyn_shoulderpad_l"(
    ComponentPose=@n14
    Alpha=Get ArmRotation_L() -> (ArmRotation_L:double)
    Translation="1,0.25,-1.5"
    Rotation="10,-0.000,8.000"
    Scale="1,1,1"
   ) {
    rotationMode=BMM_Additive
    rotationSpace=BCS_BoneSpace
    translationMode=BMM_Additive
    translationSpace=BCS_BoneSpace
    alphaBoolBlend {blendOption=Linear}
    alphaScaleBiasClamp {bInterpResult=true, interpSpeedDecreasing=20, interpSpeedIncreasing=20}
    boneToModify {boneName=dyn_shoulderpad_l}
    rotation {pitch=10, roll=8}
    translation=<1 0.25 -1.5>
   }
   n16 = Get Skydive_Armor_Alpha()
   n17 = "Transform (Modify) Bone - Bone: dyn_shoulderpad_r"(ComponentPose=@n15, Alpha=@n16, Translation="7,-0.000,0.000", Rotation="30,0.000,0.000", Scale="1,1,1") {
    rotationMode=BMM_Additive
    rotationSpace=BCS_BoneSpace
    translationMode=BMM_Additive
    translationSpace=BCS_BoneSpace
    alphaBoolBlend {blendOption=Linear}
    alphaScaleBiasClamp {bInterpResult=true, interpSpeedDecreasing=20, interpSpeedIncreasing=20}
    boneToModify {boneName=dyn_shoulderpad_r}
    rotation {pitch=30}
    translation {x=7}
   }
   n18 = "Transform (Modify) Bone - Bone: dyn_shoulderpad_l"(ComponentPose=@n17, Alpha=@n16, Translation="7,0,0", Rotation="20,0,-5", Scale="1,1,1") {
    rotationMode=BMM_Additive
    rotationSpace=BCS_BoneSpace
    translationMode=BMM_Additive
    translationSpace=BCS_BoneSpace
    alphaBoolBlend {blendOption=Linear}
    alphaScaleBiasClamp {bInterpResult=true, interpSpeedDecreasing=20, interpSpeedIncreasing=20}
    boneToModify {boneName=dyn_shoulderpad_l}
    rotation {pitch=20, roll=-5}
    translation {x=7}
   }
   n19 = "Transform (Modify) Bone - Bone: dyn_shield_parent"(
    ComponentPose=@n18
    Alpha=Get IsNotCrouching() -> (IsNotCrouching:double)
    Translation="-0.000,1.189,-0.000"
    Rotation="-2.375,-6.876,-19.410"
    Scale="1,1,1"
   ) {
    rotationMode=BMM_Additive
    rotationSpace=BCS_BoneSpace
    translationMode=BMM_Additive
    translationSpace=BCS_BoneSpace
    alphaBoolBlend {blendOption=Linear}
    alphaScaleBiasClamp {bInterpResult=true}
    boneToModify {boneName=dyn_shield_parent}
    rotation=<rot -2.375 -6.876 -19.41>
    translation {y=1.189}
   }
   n20 = "Transform (Modify) Bone - Bone: dyn_shield_parent"(ComponentPose=@n19, Alpha=Get CrouchAlpha() -> (CrouchAlpha:double), Translation="0,3,0", Rotation="0.000,0.000,-30", Scale="1,1,1") {
    rotationMode=BMM_Additive
    rotationSpace=BCS_BoneSpace
    translationMode=BMM_Additive
    translationSpace=BCS_BoneSpace
    alphaBoolBlend {blendOption=Linear}
    alphaScaleBiasClamp {bInterpResult=true}
    boneToModify {boneName=dyn_shield_parent}
    rotation {roll=-30}
    translation {y=3}
   }
   n21 = RigidBody(ComponentPose=@n20, Alpha=Get Skydive_Alpha() -> (Skydive_Alpha:double)) {
    bClampLinearTranslationLimitToRefPose=true
    bOverrideWorldGravity=true
    overridePhysicsAsset=&/F_MED_ExoSuit_Physics
    simulationSpace=BaseBoneSpace
    alphaBoolBlend {blendOption=Linear}
    baseBoneRef {boneName=pelvis}
    componentAppliedLinearAccClamp=<3000 3000 3000>
    componentLinearAccScale=<0.15 0.15 0.15>
    componentLinearVelScale=<0.15 0.15 0.15>
    overrideWorldGravity {z=-980}
   }
   Output Pose(Result=Component To Local(ComponentPose=@n21)) {layerGroup=None}
  ]}
 EventGraph {nodes [
   Event BlueprintInitializeAnimation()
   > n1 = Cast To PlayerPawn_Athena(Object=TryGetPawnOwner())
   > n2 = Set Pawn Owner(Pawn Owner:Player Pawn Athena Object Reference=@n1) -> (Output_Get:Player Pawn Athena Object Reference)
   > Set Body_Mesh(Body_Mesh=GetSkeletalMeshForPartType(self=@n2))
   Event BlueprintUpdateAnimation()
   > n3 = Sequence!(execute) -> (then_0, then_1, then_2:Exec, then_3:Exec, then_4:Exec, then_5:Exec, then_6:Exec, then_7:Exec)
   > Set CrouchAlpha(CrouchAlpha:double=SelectFloat(A=1, B=0, bPickA=Get bIsCrouching())) -> (Output_Get:double)
   n4 = NOT Boolean(A=Get bIsCrouching())
   n5 = "To Float (Boolean)"(InBool=Get ExoSuitCape_IsEquipped())
   > Set IsNotCrouching(IsNotCrouching:double=float * float(A=float - float(A=1, B=@n5), B=SelectFloat(A=1, B=0, bPickA=@n4)) {is=title K2Node_PromotableOperator}) -> (Output_Get:double)
   Set ArmRotation_R(
    execute=@n3.then_1
    ArmRotation_R:double=GetAlphaFromReferenceTransformDelta(
     SkeletalMeshComponent=Get Body_Mesh()
     BoneName=upperarm_l
     BaseName=spine_05
     DeltaComponentType=QuaternionTwist
     InputRange=(X=-174,Y=130)
     TwistAxis="0,0,1"
    )
   ) -> (Output_Get:double)
   > Set ShoulderRotation_L(ShoulderRotation_L:double=GetAlphaFromReferenceTransformDelta(
     SkeletalMeshComponent=Get Body_Mesh()
     BoneName=clavicle_l
     BaseName=spine_05
     DeltaComponentType=QuaternionTwist
     InputRange=(X=17,Y=30)
     TwistAxis="0,1,0"
    )) -> (Output_Get:double)
   Set ArmRotation_L(
    execute=@n3.then_2
    ArmRotation_L:double=GetAlphaFromReferenceTransformDelta(
     SkeletalMeshComponent=Get Body_Mesh()
     BoneName=upperarm_l
     BaseName=spine_05
     DeltaComponentType=QuaternionTwist
     InputRange=(X=-170,Y=35)
     TwistAxis="0,0,1"
    )
   ) -> (Output_Get:double)
   Set Skydive_Alpha(execute=@n3.then_3, Skydive_Alpha:double=SelectFloat(A=0.5, B=1, bPickA=Get bIsSkyDiving())) -> (Output_Get:double)
   Set Skydive_Armor_Alpha(execute=@n3.then_5, Skydive_Armor_Alpha:double=SelectFloat(A=1, B=0, bPickA=Get bIsSkyDiving())) -> (Output_Get:double)
   Set Alt_weapon(execute=@n3.then_4, Alt_weapon:bool=false) -> (Output_Get:bool)
   > n6 = Switch on EFortWeaponCoreAnimation!(execute, Selection:EFortWeaponCoreAnimation Enum=Get WeaponCoreAnim!() -> (WeaponCoreAnim)) -> (
    Melee:Exec
    Pistol:Exec
    Shotgun:Exec
    PaperBlueprint:Exec
    Rifle:Exec
    MeleeOneHand:Exec
    MachinePistol:Exec
    RocketLauncher:Exec
    GrenadeLauncher:Exec
    GoingCommando:Exec
    AssaultRifle:Exec
    TacticalShotgun:Exec
    SniperRifle:Exec
    TrapPlacement:Exec
    ShoulderLauncher:Exec
    AbilityDecoTool:Exec
    Crossbow:Exec
    C4:Exec
    RemoteControl:Exec
    DualWield:Exec
    AR_BullPup:Exec
    AR_ForwardGrip:Exec
    MedPackPaddles:Exec
    SMG_P90:Exec
    AR_DrumGun:Exec
    Consumable_Small:Exec
    Consumable_Large:Exec
    Balloon:Exec
    MountedTurret:Exec
    CreativeTool:Exec
    ExplosiveBow:Exec
    AshtonIndigo:Exec
    AshtonChicago:Exec
    MeleeDualWield:Exec
    Unarmed:Exec
   )
   n7 = Branch(execute=[@n6.AssaultRifle, @n6.Pistol, @n6.Rifle, @n6.Shotgun, @n6.TacticalShotgun])
   > Set Alt_weapon(Alt_weapon=true)
   Set Alt_weapon(execute=@n7.else, Alt_weapon=false)
   Set GliderPose_Alpha(execute=@n3.then_6, GliderPose_Alpha:double=SelectFloat(A=1, B=0, bPickA=Get bIsParachuteOpened())) -> (Output_Get:double)
   Set IsInFrontend(execute=@n3.then_7, IsInFrontend:double=SelectFloat(A=1, B=0, bPickA=Get bIsInFrontEnd())) -> (Output_Get:double)
   Event OnCharacterCustomizationLoaded()
   > n8 = Branch(Condition=BackpackItemDefinitionHasAnyMetaTag!(Tags:Array of Gameplay Tag Structures=Make Array("[0]"=(TagName="Cosmetics.MetaTags.ExoSuitCape")), bRequiresEquipped:bool=false) -> (ReturnValue:bool))
   > Set ExoSuitCape_IsEquipped(ExoSuitCape_IsEquipped:bool=true) -> (Output_Get:bool)
   > n9 = Get Dynamic Material Instance(SkeletalMeshComponent=GetPartSkeletalMeshComponent())
   > SetScalarParameterValue(self=@n9, ParameterName=Hide Element 06, Value=1)
   Set ExoSuitCape_IsEquipped(execute=@n8.else, ExoSuitCape_IsEquipped=false)
  ]}
]
variables [
 Alt_weapon {type=boolean}
 ArmRotation_L {type=number}
 ArmRotation_R {type=number}
 Body_Mesh {type=SkeletalMeshComponent}
 CrouchAlpha {type=number}
 ExoSuitCape_IsEquipped {type=boolean}
 GliderPose_Alpha {type=number}
 IsInFrontend {type=number}
 IsNotCrouching {type=number}
 Pawn Owner {type=PlayerPawn_Athena_C}
 ShoulderRotation_L {type=number}
 Skydive_Alpha {type=number}
 Skydive_Armor_Alpha {type=number}
]
