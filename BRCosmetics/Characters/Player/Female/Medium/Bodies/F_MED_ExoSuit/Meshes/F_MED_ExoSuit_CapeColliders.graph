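# Physics asset of cape colliders for the female medium ExoSuit player body, judging by its name.
# Capsules and one sphere cover pelvis, spine, clavicles, head and feet, while limb and neck bodies have no shapes.
# Named capsules on front and back dyn_rope chains and the shield parent add accessory collision.
# Free-motion constraints link only the core skeleton, leaving rope and shield bodies unconstrained.
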PhysicsAsset /BRCosmetics/Characters/Player/Female/Medium/Bodies/F_MED_ExoSuit/Meshes/F_MED_ExoSuit_CapeColliders
bodies {
 calf_l []
 calf_r []
 clavicle_l [{shapeType=Capsule, length=7.6964, radius=6.3853, center=<4.0147 -0.8354 -1.7419>, rotation=<rot -77.5566 177.3915 7.7314>}]
 clavicle_r [{shapeType=Capsule, length=7.6964, radius=6.3853, center=<-4.0147 0.8354 1.7419>, rotation=<rot -77.5565 177.3915 -172.2685>}]
 dyn_rope_01_bk [{
   shapeType=Capsule
   length=5
   radius=2.75
   shapeName=dyn_rope_01_bk_capsule
   center=<2.0873 0.2256 -0.3897>
   rotation=<rot 59.8075 -83.9369 -84.754>
  }]
 dyn_rope_01_frnt [{shapeType=Capsule, length=10, radius=3.25, shapeName=dyn_rope_01_frnt_capsule, center=<0 0.1856 0>, rotation=<rot 90 0 0>}]
 dyn_rope_02_bk [{
   shapeType=Capsule
   length=8.25
   radius=2.75
   shapeName=dyn_rope_02_bk_capsule
   center=<4.3865 0.1118 0.8617>
   rotation=<rot 78.8847 -179.9992 -167.9687>
  }]
 dyn_rope_02_frnt [{
   shapeType=Capsule
   length=6.5
   radius=3.25
   shapeName=dyn_rope_02_frnt_capsule
   center=<2.6405 0.8316 0.4251>
   rotation=<rot 90 -20.556 -42.6565>
  }]
 dyn_rope_03_frnt [{shapeType=Capsule, length=6, radius=3.25, shapeName=dyn_rope_03_frnt_capsule, rotation=<rot 0 0 90>}]
 dyn_rope_parent_bk [{shapeType=Capsule, length=6.25, radius=2.75, shapeName=dyn_rope_parent_bk_capsule, center=<2.1324 0 0.095>, rotation=<rot -87.4499 0 0>}]
 dyn_rope_parent_frnt [{shapeType=Capsule, length=15, radius=3.25, shapeName=dyn_rope_parent_frnt_capsule, rotation=<rot 90 0 0>}]
 dyn_shield_parent [{
   shapeType=Capsule
   length=1.75
   radius=8
   shapeName=dyn_shield_parent_capsule
   center=<-6.6935 -0.604 2.145>
   rotation=<rot -15.6926 -85.6438 88.8202>
  }]
 foot_l [{shapeType=Capsule, length=18.8482, radius=6.5127, center=<-4.0069 6.0073 -0.3379>, rotation=<rot 0.0388 1.4701 -89.3982>}]
 foot_r [{shapeType=Capsule, length=17.8482, radius=6.5127, center=<3.6053 -6.0175 0.3377>, rotation=<rot 0.0388 1.47 90.6018>}]
 head [{shapeType=Capsule, length=0, radius=18, center=<9.0697 0.0746 0>, rotation=<rot 0 -90 0>}]
 lowerarm_l []
 lowerarm_r []
 neck_02 []
 pelvis [
  {shapeType=Sphere, radius=16.5, shapeName=pelvis_sphere, center=<-0.3447 2.8087 -5.8545>}
  {shapeType=Capsule, length=9.4028, radius=10.6596, center=<-1.9544 -2.2142 -0.198>}
 ]
 spine_01 [{shapeType=Capsule, length=9.0171, radius=9.7354, center=<-2.8183 -1.0873 0>}]
 spine_02 [
  {shapeType=Capsule, length=5.2984, radius=8.4854, center=<2.334 -3.0231 -0.0001>}
  {shapeType=Capsule, length=20.6875, radius=9.3125, center=<-1.8317 1.9258 0>, rotation=<rot 90 0 0>}
 ]
 spine_03 [
  {shapeType=Capsule, length=3.2984, radius=8.4854, center=<-0.0747 0.5084 0>}
  {shapeType=Capsule, length=19.4688, radius=10.5312, center=<-1.2157 -2.1435 0.0002>, rotation=<rot 90 174.6442 178.7114>}
 ]
 spine_04 [
  {shapeType=Capsule, length=6.5796, radius=10.3916, center=<1.1269 -4.5139 0.0001>}
  {shapeType=Capsule, length=10.0312, radius=9.2812, center=<-1.2326 1.308 0>, rotation=<rot 90 0 0>}
 ]
 spine_05 [{shapeType=Capsule, length=10.2359, radius=10.8291, center=<-0.4236 -3.2485 0.0002>}]
 thigh_l []
 thigh_r []
 upperarm_l []
 upperarm_r []
}
constraints [
 {
  bone1Name=spine_01
  bone2Name=pelvis
  swing1LimitDegrees=45
  swing1Motion=Free
  swing2LimitDegrees=45
  swing2Motion=Free
  twistLimitDegrees=45
  twistMotion=Free
 }
 {
  bone1Name=spine_02
  bone2Name=spine_01
  swing1LimitDegrees=45
  swing1Motion=Free
  swing2LimitDegrees=45
  swing2Motion=Free
  twistLimitDegrees=45
  twistMotion=Free
 }
 {
  bone1Name=spine_03
  bone2Name=spine_02
  swing1LimitDegrees=45
  swing1Motion=Free
  swing2LimitDegrees=45
  swing2Motion=Free
  twistLimitDegrees=45
  twistMotion=Free
 }
 {
  bone1Name=spine_04
  bone2Name=spine_03
  swing1LimitDegrees=45
  swing1Motion=Free
  swing2LimitDegrees=45
  swing2Motion=Free
  twistLimitDegrees=45
  twistMotion=Free
 }
 {
  bone1Name=spine_05
  bone2Name=spine_04
  swing1LimitDegrees=45
  swing1Motion=Free
  swing2LimitDegrees=45
  swing2Motion=Free
  twistLimitDegrees=45
  twistMotion=Free
 }
 {
  bone1Name=clavicle_l
  bone2Name=spine_05
  swing1LimitDegrees=45
  swing1Motion=Free
  swing2LimitDegrees=45
  swing2Motion=Free
  twistLimitDegrees=45
  twistMotion=Free
 }
 {
  bone1Name=upperarm_l
  bone2Name=clavicle_l
  swing1LimitDegrees=45
  swing1Motion=Free
  swing2LimitDegrees=45
  swing2Motion=Free
  twistLimitDegrees=45
  twistMotion=Free
 }
 {
  bone1Name=lowerarm_l
  bone2Name=upperarm_l
  swing1LimitDegrees=45
  swing1Motion=Free
  swing2LimitDegrees=45
  swing2Motion=Free
  twistLimitDegrees=45
  twistMotion=Free
 }
 {
  bone1Name=neck_02
  bone2Name=spine_05
  swing1LimitDegrees=45
  swing1Motion=Free
  swing2LimitDegrees=45
  swing2Motion=Free
  twistLimitDegrees=45
  twistMotion=Free
 }
 {
  bone1Name=head
  bone2Name=neck_02
  swing1LimitDegrees=45
  swing1Motion=Free
  swing2LimitDegrees=45
  swing2Motion=Free
  twistLimitDegrees=45
  twistMotion=Free
 }
 {
  bone1Name=calf_l
  bone2Name=thigh_l
  swing1LimitDegrees=45
  swing1Motion=Free
  swing2LimitDegrees=45
  swing2Motion=Free
  twistLimitDegrees=45
  twistMotion=Free
 }
 {
  bone1Name=thigh_l
  bone2Name=pelvis
  swing1LimitDegrees=45
  swing1Motion=Free
  swing2LimitDegrees=45
  swing2Motion=Free
  twistLimitDegrees=45
  twistMotion=Free
 }
 {
  bone1Name=foot_l
  bone2Name=calf_l
  swing1LimitDegrees=45
  swing1Motion=Free
  swing2LimitDegrees=45
  swing2Motion=Free
  twistLimitDegrees=45
  twistMotion=Free
 }
 {
  bone1Name=clavicle_r
  bone2Name=spine_05
  swing1LimitDegrees=45
  swing1Motion=Free
  swing2LimitDegrees=45
  swing2Motion=Free
  twistLimitDegrees=45
  twistMotion=Free
 }
 {
  bone1Name=lowerarm_r
  bone2Name=upperarm_r
  swing1LimitDegrees=45
  swing1Motion=Free
  swing2LimitDegrees=45
  swing2Motion=Free
  twistLimitDegrees=45
  twistMotion=Free
 }
 {
  bone1Name=thigh_r
  bone2Name=pelvis
  swing1LimitDegrees=45
  swing1Motion=Free
  swing2LimitDegrees=45
  swing2Motion=Free
  twistLimitDegrees=45
  twistMotion=Free
 }
 {
  bone1Name=calf_r
  bone2Name=thigh_r
  swing1LimitDegrees=45
  swing1Motion=Free
  swing2LimitDegrees=45
  swing2Motion=Free
  twistLimitDegrees=45
  twistMotion=Free
 }
 {
  bone1Name=foot_r
  bone2Name=calf_r
  swing1LimitDegrees=45
  swing1Motion=Free
  swing2LimitDegrees=45
  swing2Motion=Free
  twistLimitDegrees=45
  twistMotion=Free
 }
 {
  bone1Name=upperarm_r
  bone2Name=clavicle_r
  swing1LimitDegrees=45
  swing1Motion=Free
  swing2LimitDegrees=45
  swing2Motion=Free
  twistLimitDegrees=45
  twistMotion=Free
 }
]
