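# Physics asset for the F_MED_ExoSuit face accessory mesh, covering secondary motion of hair and earrings on this female medium player body.
# A head capsule roots branching left and right hair chains, built from boxes with capsules for small strands.
# Hair constraints have narrow twist and tight swing at the roots, loosening down the main chains.
# Earring capsules attach directly to the head with free swing and twist, so their angular limit values are unused.
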PhysicsAsset /BRCosmetics/Characters/Player/Female/Medium/Bodies/F_MED_ExoSuit/Meshes/Parts/F_MED_ExoSuit_FaceAcc_Physics
bodies {
 dyn_l_earring_MSTR [{shapeType=Capsule, length=2.005, radius=0.88, shapeName=dyn_l_earring_MSTR_capsule, center=<-0.0001 0 -1.005>, rotation=<rot 0 -89.9999 0>}]
 dyn_l_hair_01 [{shapeType=Box, extentX=5, extentY=1, extentZ=4, shapeName=dyn_l_hair_01_box, center=<2 0 0>}]
 dyn_l_hair_02 [{shapeType=Box, extentX=5, extentY=1, extentZ=4, shapeName=dyn_l_hair_02_box, center=<2 0 0>}]
 dyn_l_hair_03 [{shapeType=Box, extentX=5, extentY=1, extentZ=4, shapeName=dyn_l_hair_03_box, center=<2 0 0>}]
 dyn_l_hair_MSTR [{shapeType=Box, extentX=5, extentY=1, extentZ=4, shapeName=dyn_r_hair_01_box, center=<2 0 0>}]
 dyn_l_hair_small_01 [{shapeType=Capsule, length=2, radius=0.75, shapeName=dyn_l_hair_small_01_capsule, center=<0.75 0 0>, rotation=<rot 90 0 0>}]
 dyn_r_earring_MSTR [{shapeType=Capsule, length=2.005, radius=0.88, shapeName=dyn_r_earring_MSTR_capsule, center=<-0.0001 0 -1.005>, rotation=<rot 0 -89.9999 0>}]
 dyn_r_hair_01 [{shapeType=Box, extentX=5, extentY=1, extentZ=4, shapeName=dyn_r_hair_01_box, center=<2 0 0>}]
 dyn_r_hair_02 [{shapeType=Box, extentX=5, extentY=1, extentZ=4, shapeName=dyn_r_hair_02_box, center=<2 0 0>}]
 dyn_r_hair_03 [{shapeType=Box, extentX=5, extentY=1, extentZ=4, shapeName=dyn_r_hair_03_box, center=<2 0 0>}]
 dyn_r_hair_04 [{shapeType=Box, extentX=5, extentY=1, extentZ=4, shapeName=dyn_r_hair_04_box, center=<2 0 0>}]
 dyn_r_hair_MSTR [{shapeType=Box, extentX=5, extentY=1, extentZ=4, shapeName=dyn_r_hair_01_box, center=<2 0 0>}]
 dyn_r_hair_small_01 [{shapeType=Capsule, length=2, radius=0.75, shapeName=dyn_r_hair_small_01_capsule, center=<0.75 0 0>, rotation=<rot 90 0 0>}]
 dyn_r_hair_small_02b [{shapeType=Capsule, length=2, radius=0.75, shapeName=dyn_r_hair_small_02b_capsule, center=<0.75 0 0>, rotation=<rot 90 0 0>}]
 dyn_r_hair_small_02f [{shapeType=Capsule, length=2, radius=0.75, shapeName=dyn_r_hair_small_02f_capsule, center=<0.75 0 0>, rotation=<rot 90 0 0>}]
 head [{shapeType=Capsule, length=4.9537, radius=6.2632, center=<0.8032 2.4459 0>, rotation=<rot 90 -90 -110.0027>}]
}
constraints [
 {bone1Name=dyn_l_hair_01, bone2Name=dyn_l_hair_MSTR, swing1LimitDegrees=5, swing2LimitDegrees=5, twistLimitDegrees=2}
 {bone1Name=dyn_l_hair_02, bone2Name=dyn_l_hair_01, swing1LimitDegrees=25, swing2LimitDegrees=25, twistLimitDegrees=2}
 {bone1Name=dyn_l_hair_03, bone2Name=dyn_l_hair_02, swing1LimitDegrees=25, swing2LimitDegrees=25, twistLimitDegrees=2}
 {bone1Name=dyn_l_hair_small_01, bone2Name=dyn_l_hair_01, swing1LimitDegrees=20, swing2LimitDegrees=20, twistLimitDegrees=2}
 {bone1Name=dyn_r_hair_small_02f, bone2Name=dyn_r_hair_small_01, swing1LimitDegrees=10, swing2LimitDegrees=20, twistLimitDegrees=2}
 {bone1Name=dyn_r_hair_small_02b, bone2Name=dyn_r_hair_small_01, swing1LimitDegrees=10, swing2LimitDegrees=20, twistLimitDegrees=2}
 {bone1Name=dyn_r_hair_small_01, bone2Name=dyn_r_hair_MSTR, swing1LimitDegrees=5, swing2LimitDegrees=5, twistLimitDegrees=2}
 {bone1Name=dyn_l_hair_MSTR, bone2Name=head, swing1LimitDegrees=5, swing2LimitDegrees=5, twistLimitDegrees=2}
 {bone1Name=dyn_r_hair_MSTR, bone2Name=head, swing1LimitDegrees=5, swing2LimitDegrees=5, twistLimitDegrees=2}
 {bone1Name=dyn_r_hair_01, bone2Name=dyn_r_hair_MSTR, swing1LimitDegrees=5, swing2LimitDegrees=10, twistLimitDegrees=2}
 {bone1Name=dyn_r_hair_02, bone2Name=dyn_r_hair_01, swing1LimitDegrees=20, swing2LimitDegrees=35, twistLimitDegrees=2}
 {bone1Name=dyn_r_hair_03, bone2Name=dyn_r_hair_02, swing1LimitDegrees=30, swing2LimitDegrees=35, twistLimitDegrees=2}
 {bone1Name=dyn_r_hair_04, bone2Name=dyn_r_hair_03, swing1LimitDegrees=25, swing2LimitDegrees=30, twistLimitDegrees=2}
 {
  bone1Name=dyn_l_earring_MSTR
  bone2Name=head
  swing1LimitDegrees=45
  swing1Motion=Free
  swing2LimitDegrees=45
  swing2Motion=Free
  twistLimitDegrees=45
  twistMotion=Free
 }
 {
  bone1Name=dyn_r_earring_MSTR
  bone2Name=head
  swing1LimitDegrees=45
  swing1Motion=Free
  swing2LimitDegrees=45
  swing2Motion=Free
  twistLimitDegrees=45
  twistMotion=Free
 }
]
